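# Translucent Default Lit material for Niagara sprite and mesh particles, providing smoke-style streaks for the Helios HydroBottle effect.
# Two streak textures, sampled with flipped V, have their green channels blended by Param_1.
# That mask is multiplied by particle alpha and depth-faded for soft intersections.
# A UseEmissive scalar, off by default, blends in particle color scaled by the EmissiveMult dynamic parameter.
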Material /Game/Environments/Helios/FX/HydroBottle/M_HydroBottle_LitStreaks
graph {
 properties {
  bAutomaticallySetUsageInEditor=false
  bUsedWithNiagaraMeshParticles=true
  bUsedWithNiagaraSprites=true
  blendMode=BLEND_Translucent
  pixelDepthOffsetMode=PDOM_Legacy
  usedShadingModels=Default Lit
 }
 nodes [
  Add()
  n1 = MaterialFunctionCall(Float2=TextureCoordinate()) {materialFunction=&BreakOutFloat2Components}
  n2 = MaterialFunctionCall(X=@n1.R, Y=OneMinus(None=@n1.G)) {materialFunction=&MakeFloat2}
  n3 = TextureSample(UVs=@n2.Result) {samplerType=SAMPLERTYPE_LinearColor, texture=/Game/Effects/Explosion_Update/Textures/T_Smoke_Streak}
  n4 = TextureSample(UVs=@n2.Result) {texture=/Game/Effects/Fort_Effects/Gameplay/SmokeDustExplosions/Textures/T_Explosion_Streak_02}
  n5 = ScalarParameter() {parameterName=Param_1}
  DepthFade(
   Opacity=Multiply(A=ParticleColor().A, B=LinearInterpolate(A=@n3.G, B=@n4.G, Alpha=@n5))
   FadeDistance=ScalarParameter() {defaultValue=10, parameterName=FadeDistance}
  )
  LinearInterpolate(
   B=Multiply(A=ParticleColor().RGB, B=DynamicParameter() {paramNames [EmissiveMult,Param2,Param3,Param4]}.EmissiveMult)
   Alpha=ScalarParameter() {parameterName=UseEmissive}
  )
  Multiply()
  ScalarParameter() {defaultValue=0.5}
 ]
}
