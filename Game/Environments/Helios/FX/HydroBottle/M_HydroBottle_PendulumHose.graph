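# Default Lit material for the pendulum hose of the Helios HydroBottle effect, flagged for use on Niagara ribbons.
# It wraps the U coordinate with Fmod and feeds it into a SphereMask to form repeating, parameter-driven bands along one axis.
# That mask multiplies a dark grey color parameter and also serves as the bump for PerturbNormalLQ, whose world normal is transformed to tangent space, producing procedural ridges.
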Material /Game/Environments/Helios/FX/HydroBottle/M_HydroBottle_PendulumHose
graph {
 properties {
  bAutomaticallySetUsageInEditor=false
  bUsedWithNiagaraRibbons=true
  pixelDepthOffsetMode=PDOM_Legacy
  usedShadingModels=Default Lit
  displacementFadeRange {endSizePixels=1, startSizePixels=0.0001}
 }
 nodes [
  n1 = TextureCoordinate()
  n2 = SphereMask(
   A=Fmod(A=ComponentMask(None=@n1) {r=true}, B=ScalarParameter() {defaultValue=0.01})
   B=ScalarParameter() {defaultValue=0.02, parameterName=Param_2}
   Radius=ScalarParameter() {defaultValue=0.05, parameterName=Param_3}
   Hardness=ScalarParameter() {parameterName=Param_4}
  )
  Multiply(A=VectorParameter() {defaultValue=<rgba 0.0688 0.0688 0.0688 0>}.RGB, B=@n2)
  ScalarParameter() {defaultValue=0.4, parameterName=Param_1}
  TextureSample()
  Transform(None=MaterialFunctionCall!(Bump=@n2) {materialFunction=/Engine/Functions/Engine_MaterialFunctions03/Procedurals/PerturbNormalLQ}.WS Normal) {transformSourceType=TRANSFORMSOURCE_World, transformType=TRANSFORM_Tangent}
 ]
}
