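# Translucent, volumetric per-vertex directional lit smoke material for Niagara sprites in the Helios HydroBottle effect.
# Normals blend a flattened 8x8 falling-mist flipbook normal with a flattened spherical normal tiled eight times, matching the flipbook grid to round out each frame.
# An opacity chain multiplies particle alpha by the mist texture's blue channel, then applies depth and near-camera fades.
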Material /Game/Environments/Helios/FX/HydroBottle/M_SimpleLitSmoke_HydroBottle
graph {
 properties {
  bAutomaticallySetUsageInEditor=false
  bUsedWithNiagaraSprites=true
  blendMode=BLEND_Translucent
  pixelDepthOffsetMode=PDOM_Legacy
  translucencyLightingMode=TLM_VolumetricPerVertexDirectional
  usedShadingModels=Default Lit
 }
 nodes [
  n1 = Add(A=TextureCoordinate())
  n2 = TextureCoordinate() {uTiling=8, vTiling=8}
  MaterialFunctionCall(
   BaseNormal=MaterialFunctionCall(
    Normal=TextureSample(UVs=@n1) {samplerType=SAMPLERTYPE_Normal, texture=/Game/Effects/Fort_Effects/Textures/Smoke/T_FallingMist_8x8_01_N}.RGB
    Flatness=ScalarParameter() {defaultValue=0.1}
   ) {materialFunction=&FlattenNormal}.Result
   AdditionalNormal=MaterialFunctionCall(
    Normal=TextureSample(UVs=@n2) {samplerType=SAMPLERTYPE_Normal, texture=/Game/Gadgets/Assets/Teleporter/Textures/T_Spherical_Normal_02}.RGB
    Flatness=ScalarParameter() {defaultValue=0.75, parameterName=Flatten Sphere Normal}
   ) {materialFunction=&FlattenNormal}.Result
  ) {materialFunction=&BlendAngleCorrectedNormals}
  n3 = ParticleColor()
  n4 = TextureSample() {samplerType=SAMPLERTYPE_LinearColor, texture=/Game/Effects/Fort_Effects/Textures/Smoke/T_FallingMist_8x8_01}
  Multiply(
   A=DepthFade(Opacity=Multiply(A=@n3.A, B=@n4.B), FadeDistance=ScalarParameter() {defaultValue=100, parameterName=Fade Distance})
   B=MaterialFunctionCall(Distance=ScalarParameter() {defaultValue=500, parameterName=Camera Fade}) {materialFunction=&Near_Camera_Fade}.Result
  )
  Multiply(A=SceneColor!(UVs=ScreenPosition().PixelPosition).RGB)
 ]
}
