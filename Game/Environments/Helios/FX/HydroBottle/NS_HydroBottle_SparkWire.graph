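# Niagara effect for a dangling, sparking wire in the Helios HydroBottle environment.
# Hose builds a ribbon from a self-reading chain of pendulum-constrained particles, pushed by player-position point force and curl noise that strengthen along the chain.
# Three "Smoke" emitters read the chain's end particle, emitting GPU colliding sparks, CPU sparks, and a brief glow with an arc sound, gated by loop age.
# PouringFog001 is disabled.
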NiagaraSystem /Game/Environments/Helios/FX/HydroBottle/NS_HydroBottle_SparkWire
dependencies {
 usedDataInterfaces [
  &NiagaraDataInterfaceParticleRead
  &NiagaraDataInterfaceCurve
  &NiagaraDataInterfaceSpriteRendererInfo
  &NiagaraDataInterfaceColorCurve
  &NiagaraDataInterfaceCamera
 ]
 usedDynamicInputs []
 usedModules [
  &SystemState
  &self:SystemSpawnScript.SystemScriptSource.SystemScriptGraph.NiagaraNodeAssignment_1.SetVariables_#1
  &self:SystemSpawnScript.SystemScriptSource.SystemScriptGraph.NiagaraNodeAssignment_0.SetVariables_#2
 ]
}
emitters {
 Hose {
  emitterUpdateScript [
   &EmitterState(Life Cycle Mode=::NewEnumerator1, Loop Behavior=::NewEnumerator1, Loop Duration=1, UseLoopDelay=false)
   /SetVariables(Emitter.RopeAttributeReader=DataInterface() {propertyValues {EmitterBinding {bindingMode=Self, emitterName=Hose}}}) {isSetParameters=true}
   &SpawnBurst_Instantaneous(Spawn Count=12, Use Spawn Probability=false, UseLoopCountLimit=false)
  ]
  particleSpawnScript [&InitializeParticle(Ribbon Width=4, Ribbon Width Mode=::NewEnumerator1, UsePositionOffset=false, Write Material Random=true)]
  particleUpdateScript [
   &ParticleState(Kill Particles When Lifetime Has Elapsed=false)
   &self:FindTip(ParticleLength=12) {module=FindTip}
   &PendulumSetup(
    Module.Enable Spring Driver=true
    Pendulum Length=10
    Pendulum Pivot:Vector3f=$Transient.TipPosition
    "Spring Driven Constraint (Tightness)"=50
   )
   &PointForce(
    Force Falloff Distance=1000
    Force Origin=&MakeVectorFromLinearColorRGB(Color=$NPC.FortniteNPC.PlayerPosition)
    Force Strength=&Lerp_Float(A=0, B=175000, Alpha=&ReturnNormalizedExecIndex())
    Module.Use Falloff=true
    Module.UseLinearForce=false
   )
   &CurlNoiseForce(
    Bias Noise=false
    Noise Frequency=10
    Noise Quality / Cost=::NewEnumerator2
    Noise Strength=&Multiply_Float(
     A=&Lerp_Float(A=0, B=350000, Alpha=&ReturnNormalizedExecIndex())
     B=&FloatFromCurve(
      FloatCurve=DataInterface() {propertyValues {Curve {keys [
          {interpMode=RCIM_Cubic, tangentMode=RCTM_Break, time=0, value=-0.0275}
          {interpMode=RCIM_Cubic, tangentMode=RCTM_Break, time=0.025, value=0.25}
          {interpMode=RCIM_Cubic, tangentMode=RCTM_Break, time=0.05, value=0}
         ]}}}
      CurveIndex=$System.NormalizedLoopAge
     )
    )
    Pan Noise=true
    Pan Noise Field=<0 0 -1>
   )
   &Drag(Use Linear Drag=true, Use Rotational Drag=true)
   &CurlNoiseForce(
    Bias Noise=false
    Noise Frequency=5
    Noise Strength=&Lerp_Float(A=0, B=35000, Alpha=&ReturnNormalizedExecIndex())
    Pan Noise=true
    Pan Noise Field=<0 0 -1>
   ) {module=CurlNoiseForce001}
   &SolveForcesAndVelocity(Clamp Velocity=false, Limit Acceleration=false)
   &PendulumConstraint(Calculate and Output Potential Energy=true)
   &CalculateAccurateVelocity(Calculate Accurate Velocity On First Frame=true, Previous Position:Vector3f=$Transient.TipPosition)
   &KillParticles(Kill Particles=&SetBoolByIntComparison(A=&ReturnParticleIDAttributes(), B=11, Comparison Type=::NewEnumerator0))
  ]
  rendererClasses [&NiagaraRibbonRendererProperties]
  renderers [{rendererClass=&NiagaraRibbonRendererProperties}]
 }
 PouringFog001 {
  bEnabled=false
  emitterUpdateScript [
   &EmitterState()
   &SpawnRate(SpawnRate=5, Use Spawn Probability=false)
  ]
  particleSpawnScript [
   &InitializeParticle(
    Color=&RandomRangeLinearColor(Minimum=<rgba 0.3072 0.4519 0.4979 0.125>, Maximum=<rgba 0.3774 0.6731 0.6812 0.75>)
    Lifetime Max=8
    Lifetime Min=4
    Lifetime Mode=::NewEnumerator1
    Position Offset=<0 0 -750>
    Sprite Rotation Mode=::NewEnumerator2
    Sprite Size Mode=::NewEnumerator1
    Sprite UV Mode=::NewEnumerator3
    Uniform Sprite Size Max=500
    Uniform Sprite Size Min=250
    UsePositionOffset=true
    Write Material Random=true
   )
   &SubUVAnimation(
    "Sprite Renderer "=DataInterface() {propertyValues {SpriteRenderer=&self:PouringFog001_1.NiagaraSpriteRendererProperties_0}}
    UseEndFrame=false
    UseStartFrame=false
   )
   &ShapeLocation(Box Size=<750 750 0>, MaskSpawnGroup=false, Override Seed=false, Shape Primitive=::NewEnumerator2, Surface Only=false, UseOwnerScale=true)
   &AddVelocity(Velocity=&RandomRangeVector(Minimum=<-50 -50 0>, Maximum=<50 50 0>))
  ]
  particleUpdateScript [
   &Drag(Drag=4, Use Linear Drag=true, Use Rotational Drag=true)
   &SolveForcesAndVelocity(Clamp Velocity=false, Limit Acceleration=false)
   &ParticleState()
   &SpriteRotationRate(Rotation Rate=15)
   &ScaleSpriteSize(Uniform Curve Sprite Scale=DataInterface() {propertyValues {Curve {keys [{interpMode=RCIM_Cubic, leaveTangent=2.4139, tangentMode=RCTM_Break, time=0, value=0.4}, {tangentMode=RCTM_User, time=1, value=1}]}}})
   &ScaleColor(
    Linear Color Curve=DataInterface() {propertyValues {
      AlphaCurve {keys [<at 0.0038 0.1>,<at 0.3087 1>,<at 1 0>]}
      BlueCurve {keys [<at 0 1>,<at 1 1>]}
      GreenCurve {keys [<at 0 1>,<at 1 1>]}
      RedCurve {keys [<at 0 1>,<at 1 1>]}
     }}
    Scale Mode=::NewEnumerator2
   )
   &DynamicMaterialParameters(
    Index 0 Param 1=-0.5
    Module.Index0Param0Write=true
    Module.Index0Param1Write=true
    Module.Index0Param2Write=true
    Module.Index0Param3Write=true
   )
   &ScaleSpriteSize(
    Scale Sprite Size Mode=::NewEnumerator0
    Uniform Scale Factor=&Lerp_Float(
     A=0
     B=1
     Alpha=&self:ScratchDynamicInput_03(
      End Distance:NiagaraFloat=1250
      Max Output:NiagaraFloat=1
      MCP Player Location:LinearColor=&MakeLinearColorFromVectorAndFloat(
       "Vector (RGB)"=&ReturnCameraPosition(
        Coordinate Space:ENiagaraCoordinateSpace=::Simulation
        Camera Query:NiagaraDataInterfaceCamera=DataInterface() {propertyValues {bRequireCurrentFrameData=false}}
       )
       "Float (Alpha)"=1
      )
      Minimum Output:NiagaraFloat=0
      Start Distance:NiagaraFloat=0
     )
    )
   ) {module=ScaleSpriteSize001}
  ]
  rendererClasses [&NiagaraSpriteRendererProperties]
  renderers [{rendererClass=&NiagaraSpriteRendererProperties}]
 }
 Smoke {
  simTarget=GPUComputeSim
  emitterUpdateScript [
   &self:ScratchModule_02(Particle Attribute Reader=DataInterface() {propertyValues {EmitterBinding {bindingMode=Other, emitterName=Hose}}}, Particle Index=11) {module=ScratchModule_02}
   &EmitterState()
   &SpawnPerFrame(
    Spawn:NiagaraBool=&SetBoolByFloatComparison(A=0.05, B=$System.NormalizedLoopAge, Comparison Type=::NewEnumerator2)
    Spawn Count:NiagaraInt32=&RandomRangeInt(Minimum=5, Maximum=10)
    Use Spawn Probability=false
   )
  ]
  particleSpawnScript [
   &InitializeParticle(
    Color=<rgba 50 10.3715 0 1>
    Lifetime Max=1.5
    Lifetime Min=0.15
    Lifetime Mode=::NewEnumerator1
    Position=$Emitter.Position
    Position Mode=::NewEnumerator1
    Sprite Rotation Mode=::NewEnumerator2
    Sprite Size Mode=::NewEnumerator1
    Sprite UV Mode=::NewEnumerator1
    Uniform Sprite Size Min=3
    UsePositionOffset=false
    Write Material Random=true
   )
   &AddVelocity(
    Module.Use Falloff Distance=false
    Override Seed=false
    Velocity Mode=::NewEnumerator1
    Velocity Origin=$Emitter.Position
    Velocity Speed=&FloatFromCurve(
     FloatCurve=DataInterface() {propertyValues {Curve {keys [
         {interpMode=RCIM_Cubic, tangentMode=RCTM_Break, time=0, value=0}
         {interpMode=RCIM_Cubic, tangentMode=RCTM_Break, time=0.5006, value=1.0039}
         <at 1 0>
        ]}}}
     CurveIndex=&RandomRangeFloat(Minimum=0, Maximum=1)
     Scale Curve=500
    )
   )
  ]
  particleUpdateScript [
   &ParticleState()
   &ScaleSpriteSize(Uniform Curve Sprite Scale=DataInterface() {propertyValues {Curve {keys [{interpMode=RCIM_Cubic, leaveTangent=2.5018, tangentMode=RCTM_Break, time=0, value=0}, {tangentMode=RCTM_User, time=1, value=1}]}}})
   &GravityForce()
   &Drag(Use Linear Drag=true, Use Rotational Drag=true)
   &Collision(
    Age Colliding Particles=false
    Kill Offscreen Particles=true
    KillOnCollision=false
    Module.Control Rest State Via Surface Normal=false
    Module.EnableMaxCollisionCount=false
    Module.Randomize Collision Normal Vector=false
   )
   &SolveForcesAndVelocity(Clamp Velocity=false, Limit Acceleration=false)
   &ScaleSpriteSizeBySpeed(Max Scale Factor=<1 2>, Min Scale Factor=<0.05 1>)
   &ScaleColor(
    Linear Color Curve=DataInterface() {propertyValues {
      AlphaCurve {keys [<at 0.0064 0>,<at 0.1236 0.5>,<at 1 0>]}
      BlueCurve {keys [<at 0 1>,<at 1 1>]}
      GreenCurve {keys [<at 0 0.4085>,<at 1 1>]}
      RedCurve {keys [<at 0 0>,<at 1 1>]}
     }}
    Scale Mode=::NewEnumerator2
   )
  ]
  rendererClasses [&NiagaraSpriteRendererProperties]
  renderers [{rendererClass=&NiagaraSpriteRendererProperties}]
 }
 Smoke001 {
  emitterSpawnScript [&InitializeAudioPlayer(Audio Player=DataInterface() {propertyValues {
      Attenuation=null
      Concurrency=/Game/Sounds/Concurrency_Rules/ThreeMaximum_StopFurthest
      MaxPlaysPerTick=10
      SoundToPlay=/Game/Athena/Sounds/Quests/EnergyCloud/Spot_Electric_Spark_Arc_Small_02_Cue
      bAllowLoopingOneShotSounds=false
      bLimitPlaysPerTick=true
      bOnlyActiveDuringGameplay=false
      bStopWhenComponentIsDestroyed=true
      ConfigurationUserParameter {parameter {name=None, typeDefHandle {registeredTypeIndex=1}, varData []}}
      ParameterNames []
     }})]
  emitterUpdateScript [
   &self:ScratchModule_02(Particle Attribute Reader=DataInterface() {propertyValues {EmitterBinding {bindingMode=Other, emitterName=Hose}}}, Particle Index=11) {module=ScratchModule_02}
   &EmitterState()
   &SpawnPerFrame(
    Spawn:NiagaraBool=&SetBoolByFloatComparison(A=0.01, B=$System.NormalizedLoopAge, Comparison Type=::NewEnumerator2)
    Use Spawn Probability=false
   )
  ]
  particleSpawnScript [
   &InitializeParticle(
    Color=<rgba 1 0.506 0.3767 0.25>
    Lifetime=0.1
    Position=$Emitter.Position
    Position Mode=::NewEnumerator1
    Sprite Rotation Mode=::NewEnumerator2
    Sprite Size Mode=::NewEnumerator1
    Sprite UV Mode=::NewEnumerator1
    Uniform Sprite Size Max=350
    Uniform Sprite Size Min=300
    UsePositionOffset=false
    Write Material Random=true
   )
   /Niagara/Modules/Audio/PlayAudio(Pitch=1.2)
   &CameraOffset(Camera Offset Amount=150)
  ]
  particleUpdateScript [
   &ParticleState()
   &ScaleSpriteSize(Uniform Curve Sprite Scale=DataInterface() {propertyValues {Curve {keys [{interpMode=RCIM_Cubic, leaveTangent=2.5018, tangentMode=RCTM_Break, time=0, value=0}, {tangentMode=RCTM_User, time=1, value=1.5}]}}})
   &ScaleColor(
    Linear Color Curve=DataInterface() {propertyValues {
      AlphaCurve {keys [<at 0.0064 0>,<at 0.1236 0.5>,<at 1 0>]}
      BlueCurve {keys [<at 0 1>,<at 1 1>]}
      GreenCurve {keys [<at 0 1>,<at 1 0.4333>]}
      RedCurve {keys [<at 0 1>,<at 1 0>]}
     }}
    Scale Mode=::NewEnumerator2
   )
  ]
  rendererClasses [&NiagaraSpriteRendererProperties]
  renderers [{rendererClass=&NiagaraSpriteRendererProperties}]
 }
 Smoke002 {
  emitterUpdateScript [
   &self:ScratchModule_02(Particle Attribute Reader=DataInterface() {propertyValues {EmitterBinding {bindingMode=Other, emitterName=Hose}}}, Particle Index=11) {module=ScratchModule_02}
   &EmitterState()
   &SpawnPerFrame(
    Spawn:NiagaraBool=&SetBoolByFloatComparison(A=0.05, B=$System.NormalizedLoopAge, Comparison Type=::NewEnumerator2)
    Spawn Count:NiagaraInt32=&RandomRangeInt(Minimum=10, Maximum=20)
    Use Spawn Probability=false
   )
  ]
  particleSpawnScript [
   &InitializeParticle(
    Color=<rgba 50 10.3715 0 1>
    Lifetime Max=1.5
    Lifetime Min=0.15
    Lifetime Mode=::NewEnumerator1
    Position=$Emitter.Position
    Position Mode=::NewEnumerator1
    Sprite Rotation Mode=::NewEnumerator2
    Sprite Size Mode=::NewEnumerator1
    Sprite UV Mode=::NewEnumerator1
    Uniform Sprite Size Min=3
    UsePositionOffset=false
    Write Material Random=true
   )
   &AddVelocity(
    Module.Use Falloff Distance=false
    Override Seed=false
    Velocity Mode=::NewEnumerator1
    Velocity Origin=$Emitter.Position
    Velocity Speed=&FloatFromCurve(
     FloatCurve=DataInterface() {propertyValues {Curve {keys [
         {interpMode=RCIM_Cubic, tangentMode=RCTM_Break, time=0, value=0}
         {interpMode=RCIM_Cubic, tangentMode=RCTM_Break, time=0.5006, value=1.0039}
         <at 1 0>
        ]}}}
     CurveIndex=&RandomRangeFloat(Minimum=0, Maximum=1)
     Scale Curve=500
    )
   )
  ]
  particleUpdateScript [
   &ParticleState()
   &ScaleSpriteSize(Uniform Curve Sprite Scale=DataInterface() {propertyValues {Curve {keys [{interpMode=RCIM_Cubic, leaveTangent=2.5018, tangentMode=RCTM_Break, time=0, value=0}, {tangentMode=RCTM_User, time=1, value=1}]}}})
   &GravityForce()
   &Drag(Use Linear Drag=true, Use Rotational Drag=true)
   &SolveForcesAndVelocity(Clamp Velocity=false, Limit Acceleration=false)
   &ScaleSpriteSizeBySpeed(Max Scale Factor=<1 2>, Min Scale Factor=<0.05 1>)
   &ScaleColor(
    Linear Color Curve=DataInterface() {propertyValues {
      AlphaCurve {keys [<at 0.0064 0>,<at 0.1236 0.5>,<at 1 0>]}
      BlueCurve {keys [<at 0 1>,<at 1 1>]}
      GreenCurve {keys [<at 0 1>,<at 1 1>]}
      RedCurve {keys [<at 0 1>,<at 1 1>]}
     }}
    Scale Mode=::NewEnumerator2
   )
  ]
  rendererClasses [&NiagaraSpriteRendererProperties]
  renderers [{rendererClass=&NiagaraSpriteRendererProperties}]
 }
}
summary {
 systemName=NS_HydroBottle_SparkWire
 emitters [
  {emitterName=Hose, rendererClasses [&NiagaraRibbonRendererProperties]}
  {bEnabled=false, emitterName=PouringFog001, rendererClasses [&NiagaraSpriteRendererProperties]}
  {emitterName=Smoke, simTarget=GPUComputeSim, rendererClasses [&NiagaraSpriteRendererProperties]}
  {emitterName=Smoke001, rendererClasses [&NiagaraSpriteRendererProperties]}
  {emitterName=Smoke002, rendererClasses [&NiagaraSpriteRendererProperties]}
 ]
 userVariables [
  User.Audio Player {
   type=NiagaraDataInterfaceAudioPlayer
   description=""
   defaultValue {
    struct=NiagaraExt_VariableValue_DataInterface
    value {
     dataInterface=&self:NiagaraDataInterfaceAudioPlayer_0
     dataInterfaceClass=&NiagaraDataInterfaceAudioPlayer
    }
   }
  }
  User.Audio Player001 {
   type=NiagaraDataInterfaceAudioPlayer
   description=""
   defaultValue {
    struct=NiagaraExt_VariableValue_DataInterface
    value {
     dataInterface=&self:NiagaraDataInterfaceAudioPlayer_1
     dataInterfaceClass=&NiagaraDataInterfaceAudioPlayer
    }
   }
  }
 ]
}
